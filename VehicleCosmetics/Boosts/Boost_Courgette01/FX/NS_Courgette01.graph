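# Niagara system for the Courgette01 vehicle boost cosmetic, trailing effects from the car's filtered boost sockets.
# A ribbon emitter and three sprite or mesh emitters spawn per-socket via partitioning, offset by DTCarOffset, with spawn groups varying shape, motion and renderer visibility.
# GameModeID indexes arrays that tune color, brightness, speed and some offsets and lifetimes.
# Curves animate size, alpha and dynamic material parameters; the painted-variant override modules are disabled.
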NiagaraSystem /VehicleCosmetics/Boosts/Boost_Courgette01/FX/NS_Courgette01
dependencies {
 usedDataInterfaces [
  &NiagaraDataInterfaceCurve
  &NiagaraDataInterfaceArrayColor
  &NiagaraDataInterfaceArrayFloat3
  &NiagaraDataInterfaceArrayFloat
  &NiagaraDataInterfaceSpriteRendererInfo
  &NiagaraDataInterfaceColorCurve
  &NiagaraDataInterfaceSkeletalMesh
 ]
 usedDynamicInputs []
 usedModules [
  &self:SystemSpawnScript.SystemScriptSource.SystemScriptGraph.NiagaraNodeAssignment_0.SetVariables_#1
  &SystemState
 ]
}
emitters {
 NE_VC_Boost_Ribbon_Parent {
  emitterUpdateScript [
   &EmitterState(Inactive Response=::NewEnumerator1, Life Cycle Mode=::NewEnumerator1, Loop Duration=1, UseLoopDelay=false)
   &VC_FilteredSocketCount()
   &VC_RateQualityHandler(Epic Spawn Rate=40, High Spawn Rate=40, Low Spawn Rate=30, Medium Spawn Rate=36, Use Custom Rate?=true)
   &VC_SpawnRibbon_v1(Number of Ribbons=$Emitter.PortCount, SpawnRate=$Emitter.Spawn Rate)
   /SetVariables(Emitter.rate:NiagaraFloat=&SelectFloatFromArray(
     Array Sampling Mode=::NewEnumerator0
     Direct Array Index=?
     Float Selection Array=DataInterface() {propertyValues {FloatData [4,5,6]}}
     Random Seed=0
    )) {isSetParameters=true}
  ]
  particleSpawnScript [
   &VC_PaintedVariantsOverride() {enabled=false}
   &VC_InitializeRibbon_v1(
    Custom Offset=true
    Enable Skeletal Position=true
    Lifetime=1
    Local Vector Offset=&AddVectorToPosition(Position=<0 0 0>, Vector=$User.DTCarOffset)
    Offset Amount=1
    Ribbon Width=60
   )
   &AddVelocity(Velocity=&SelectVectorFromArray(
     Direct Array Index=$User.GameModeID
     Vector Selection Array=DataInterface() {propertyValues {FloatData [<-2000 0 0>, <-2000 0 0>, <-1400 0 0>]}}
    )) {module=AddVelocity001}
   &DynamicMaterialParameters(
    Index 2 Param 2=&SelectFloatFromArray(Direct Array Index=$Particles.SpawnGroup, Float Selection Array=DataInterface() {propertyValues {FloatData [0,0.5]}})
    Module.Index2Float0Write=true
    Module.Index2Float1Write=true
    Module.Index2Float2Write=true
    Module.Index2Float3Write=true
    Write Parameter Index 0=::NewEnumerator0
    Write Parameter Index 2=::NewEnumerator1
   )
  ]
  particleUpdateScript [
   &ParticleState()
   &InheritSourceMovement(ApplyForce=false, ApplyPosition=true, ApplyPositionFalloff=false)
   &SolveForcesAndVelocity(Clamp Velocity=false, Limit Acceleration=false)
   &VC_UpdateRibbon_v1(Enable Skeletal Position=true)
   &ScaleColor(
    Scale RGB=&VectorFromFloat(Value=&SelectFloatFromArray(Direct Array Index=$User.GameModeID, Float Selection Array=DataInterface() {propertyValues {FloatData [2,3,3]}}))
    ScaleA=true
    ScaleRGB=true
   )
   /SetVariables(
    Particles.linescolor:LinearColor=&MultiplyLinearColor(
     Linear Color=&ColorFromCurve(ColorCurve=DataInterface() {propertyValues {
        AlphaCurve {keys [<at 0.4716 1 RCIM_Cubic>,<at 0.6651 0>]}
        BlueCurve {keys [<at 0 1 RCIM_Cubic>]}
        GreenCurve {keys [<at 0 1 RCIM_Cubic>]}
        RedCurve {keys [<at 0 1 RCIM_Cubic>]}
       }})
     Scale Factor=&MakeVector4FromVectorAndFloat(
      "Vector (RGB)"=&VectorFromFloat(Value=&SelectFloatFromArray(Direct Array Index=$User.GameModeID, Float Selection Array=DataInterface() {propertyValues {FloatData [1.5,2.5,2.5]}}))
      "Float (Alpha)"=1
     )
    )
    Particles.linessize=50
   ) {isSetParameters=true}
  ]
  rendererClasses [&NiagaraRibbonRendererProperties]
  renderers [
   {rendererClass=&NiagaraRibbonRendererProperties}
   {rendererClass=&NiagaraRibbonRendererProperties, rendererIndex=1}
  ]
 }
 NE_VC_Boost_Sprite_Parent {
  emitterUpdateScript [
   &EmitterState(Inactive Response=::NewEnumerator1, Life Cycle Mode=::NewEnumerator1, Loop Duration=0.6, UseLoopDelay=false)
   &VC_FilteredSocketCount()
   &SpawnRate(SpawnRate=14, Use Spawn Probability=false)
   &SpawnRate(Spawn Probability=0.6, SpawnGroup=1, SpawnRate=4, Use Spawn Probability=true) {module=SpawnRate001}
   &SpawnRate(Spawn Probability=0.6, SpawnGroup=2, SpawnRate=4, Use Spawn Probability=true) {module=SpawnRate002}
   &SpawnRate(Spawn Probability=0.5, SpawnGroup=3, SpawnRate=6, Use Spawn Probability=true) {module=SpawnRate003}
   /SetVariables(
    Emitter.bottom:NiagaraFloat=&FloatFromCurve(FloatCurve=DataInterface() {propertyValues {Curve {keys [<at 0.5 4>,<at 0.8 30>,<at 1 90>]}}}, CurveIndex=$Emitter.NormalizedLoopAge)
    Emitter.top:NiagaraFloat=&FloatFromCurve(FloatCurve=DataInterface() {propertyValues {Curve {keys [<at 0 20>,<at 0.5 0>]}}}, CurveIndex=$Emitter.NormalizedLoopAge)
   ) {isSetParameters=true}
  ]
  particleSpawnScript [
   &VC_PaintedVariantsOverride() {enabled=false}
   &InitializeParticle(
    Color=&SelectLinearColorFromArray(
     Array Sampling Mode=::NewEnumerator1
     Direct Array Index=$User.GameModeID
     Color Selection Array=DataInterface() {propertyValues {ColorData [<rgba 0.65 0.65 0.7 1>, <rgba 0.8357 0.8357 0.9 1>, <rgba 0.8357 0.8357 0.9 1>]}}
     Random Seed=?
    )
    Lifetime=&MergeHelper_FloatSwitch(
     0=&RandomRangeFloat(Minimum=0.8, Maximum=1)
     1=1.6
     2=1.6
     3=1.6
     Index=&SpawnGroup()
     Num=::NewEnumerator2
    )
    Mesh Scale=&Multiply_VectorByFloat(
     Vector=&MergeHelper_VectorSwitch(
      0=<0 0 0>
      1=&RandomRangeVector(Minimum=<4 1.2 1.2>, Maximum=<2 0.9 0.9>)
      2=&RandomRangeVector(Minimum=<4 1.2 1.2>, Maximum=<2 0.9 0.9>)
      3=&RandomRangeVector(Minimum=<4 1.2 1.2>, Maximum=<2 0.9 0.9>)
      Index=&SpawnGroup()
      Num=::NewEnumerator2
     )
     Float=0.85
    )
    Mesh Scale Mode=::NewEnumerator4
    Sprite Rotation Mode=::NewEnumerator2
    Sprite Size Mode=::NewEnumerator1
    Uniform Sprite Size Max=30
    Uniform Sprite Size Min=15
    UsePositionOffset=false
    Write Material Random=true
   )
   &PartitionParticles(Number Of Partitions=$Emitter.PortCount)
   &SkeletalMeshLocation(
    Bone / Socket Index=$Particles.PartitionParticles.Partition
    Bone / Socket Sampling Mode=::NewEnumerator5
    MaskSpawnGroup=false
    Mesh Sampling Type=::NewEnumerator6
    Skeletal Mesh=$System.SkeletalMesh
   )
   &ShapeLocation(
    MaskSpawnGroup=false
    Non Uniform Scale=<0.1 1 1>
    Offset=&AddVectorToPosition(
     Position=&AddVectorToPosition(
      Position=&MergeHelper_VectorSwitch(0=<-60 0 0>, 1=<60 0 0>, 2=<60 0 0>, 3=<60 0 0>, Index=&SpawnGroup(), Num=::NewEnumerator2)
      Vector=&SelectVectorFromArray(Direct Array Index=$User.GameModeID, Vector Selection Array=DataInterface() {propertyValues {FloatData [<0 0 0>, <0 0 0>, <-70 0 0>]}})
     )
     Vector=$User.DTCarOffset
    )
    Offset Coordinate Space=::Local
    Offset Mode=::NewEnumerator0
    Override Seed=false
    Sphere Radius=&MergeHelper_FloatSwitch(0=30, 1=22, 2=20, 3=21, Index=&SpawnGroup(), Num=::NewEnumerator2)
    Sphere Surface Distribution=0.2
    UseOwnerScale=true
   )
   &AddVelocity(
    Velocity=&MergeHelper_VectorSwitch(
     0=&RandomRangeVector(Minimum=<-900 0 0>, Maximum=<-1100 0 0>)
     1=&RandomRangeVector(Minimum=<-100 0 0>, Maximum=<-150 0 0>)
     2=&RandomRangeVector(Minimum=<-100 0 0>, Maximum=<-150 0 0>)
     3=&RandomRangeVector(Minimum=<-100 0 0>, Maximum=<-150 0 0>)
     Index=&SpawnGroup()
     Num=::NewEnumerator2
    )
    Velocity Speed Scale=&SelectFloatFromArray(Direct Array Index=$User.GameModeID, Float Selection Array=DataInterface() {propertyValues {FloatData [1,0.8,0.6]}})
   )
   &InitialMeshOrientation(
    Mesh Axis To Orient=<-1 0 0>
    Mesh Orientation Mode=::NewEnumerator2
    Orientation Vector=&NormalizeVector(Vector To Normalize=$Particles.Velocity, Safe Normalize=true, Safe Fallback Threshold=0, Safe Fallback Vector=<0 0 0.001>)
    Rotation=&MergeHelper_VectorSwitch(
     0=<0 0 0>
     1=&RandomRangeVector(Minimum=<-0.02 0 0>, Maximum=<0.02 0 0>)
     2=&RandomRangeVector(Minimum=<0.48 0 0>, Maximum=<0.52 0 0>)
     3=&SelectVectorFromArray(
      Array Sampling Mode=::NewEnumerator0
      Direct Array Index=?
      Vector Selection Array=DataInterface() {propertyValues {FloatData [<-0.25 0 0>, <0.25 0 0>, <-0.75 0 0>, <0.75 0 0>]}}
      Random Seed=0
     )
     Index=&SpawnGroup()
     Num=::NewEnumerator2
    )
    Rotation Coordinate Space=::NewEnumerator2
    Use Rotation Vector=true
   )
   /SetVariables(Particles.VisibilityTag=$Emitter.SpawnGroup) {isSetParameters=true}
  ]
  particleUpdateScript [
   &ParticleState()
   &ScaleSpriteSize(Uniform Curve Sprite Scale=DataInterface() {propertyValues {Curve {keys [{tangentMode=RCTM_User, time=0, value=0.5}, {tangentMode=RCTM_User, time=0.25, value=1}]}}})
   &InheritSourceMovement(ApplyForce=false, ApplyPosition=true, ApplyPositionFalloff=false)
   &SolveForcesAndVelocity(Clamp Velocity=false, Limit Acceleration=false)
   &ScaleColor(
    Scale Alpha=&FloatFromCurve(FloatCurve=DataInterface() {propertyValues {Curve {
        defaultValue=0
        keys [<at 0 0 RCIM_Cubic>, {arriveTangent=1.4502, interpMode=RCIM_Cubic, leaveTangent=1.4502, tangentMode=RCTM_User, time=1, value=1}]
       }}})
    ScaleA=true
    ScaleRGB=false
   )
   /SetVariables(Particles.dotcolor:LinearColor=&MultiplyLinearColor(
     Linear Color=&SelectLinearColorFromArray(
      Array Sampling Mode=::NewEnumerator0
      Direct Array Index=?
      Color Selection Array=DataInterface() {propertyValues {ColorData [<rgba 0.8261 0.1912 0.1006 1>, <rgba 0.7652 0.4452 0.1198 1>]}}
      Random Seed=0
     )
     Scale Factor=&MakeVector4FromVectorAndFloat(
      "Vector (RGB)"=&VectorFromFloat(Value=1)
      "Float (Alpha)"=&FloatFromCurve(
       FloatCurve=DataInterface() {propertyValues {Curve {keys [
           {arriveTangent=3.2687, interpMode=RCIM_Cubic, leaveTangent=3.2687, leaveTangentWeight=0.0667, tangentMode=RCTM_User, time=0, value=0}
           {arriveTangentWeight=0.0667, interpMode=RCIM_Cubic, leaveTangentWeight=0.1667, tangentMode=RCTM_User, time=0.2, value=1}
           {
            arriveTangentWeight=0.1453
            interpMode=RCIM_Cubic
            leaveTangent=-0.538
            leaveTangentWeight=0.2616
            tangentMode=RCTM_Break
            time=0.35
            value=1
           }
           {arriveTangent=-0.3061, arriveTangentWeight=0.1667, interpMode=RCIM_Cubic, leaveTangent=-0.3061, tangentMode=RCTM_User, time=0.9, value=0}
          ]}}}
       Scale Curve=0.4
      )
     )
    )) {isSetParameters=true}
   &DynamicMaterialParameters(
    Index 1 Param 1=&FloatFromCurve(FloatCurve=DataInterface() {propertyValues {Curve {keys [
         {arriveTangentWeight=0.0764, interpMode=RCIM_Cubic, leaveTangentWeight=0.1667, tangentMode=RCTM_User, time=0, value=1}
         {arriveTangentWeight=0.1863, interpMode=RCIM_Cubic, leaveTangentWeight=0.1863, tangentMode=RCTM_User, time=0.4, value=0}
         {arriveTangentWeight=0.1667, interpMode=RCIM_Cubic, tangentMode=RCTM_User, time=1, value=1.5}
        ]}}})
    Module.Index1Float0Write=true
    Module.Index1Float1Write=true
    Module.Index1Float2Write=true
    Module.Index1Float3Write=true
    Write Parameter Index 0=::NewEnumerator0
    Write Parameter Index 1=::NewEnumerator1
   )
  ]
  rendererClasses [&NiagaraMeshRendererProperties, &NiagaraSpriteRendererProperties]
  renderers [
   {rendererClass=&NiagaraMeshRendererProperties, rendererIndex=0}
   {rendererClass=&NiagaraMeshRendererProperties, rendererIndex=1}
   {rendererClass=&NiagaraMeshRendererProperties, rendererIndex=2}
   {rendererClass=&NiagaraSpriteRendererProperties, rendererIndex=3}
  ]
 }
 NE_VC_Boost_Sprite_Parent001 {
  emitterUpdateScript [
   &EmitterState(Inactive Response=::NewEnumerator1, Life Cycle Mode=::NewEnumerator1, Loop Duration=1, UseLoopDelay=false)
   &VC_FilteredSocketCount()
   &SpawnRate(SpawnRate=22, Use Spawn Probability=false)
   &SpawnRate(SpawnGroup=1, SpawnRate=17, Use Spawn Probability=false) {module=SpawnRate001}
  ]
  particleSpawnScript [
   &VC_PaintedVariantsOverride() {enabled=false}
   &InitializeParticle(
    Color=&SelectLinearColorFromArray(
     Array Sampling Mode=::NewEnumerator1
     Direct Array Index=$User.GameModeID
     Color Selection Array=DataInterface() {propertyValues {ColorData [<rgba 1 1 1 1>, <rgba 2 2 2 1>, <rgba 2 2 2 1>]}}
     Random Seed=?
    )
    Lifetime=&MergeHelper_FloatSwitch(
     0=&RandomRangeFloat(Minimum=0.7, Maximum=0.8)
     1=&RandomRangeFloat(Minimum=1, Maximum=1.2)
     2=&RandomRangeFloat(Minimum=0.7, Maximum=0.8)
     3=?
     Index=$User.GameModeID
     Num=::NewEnumerator1
    )
    Sprite Rotation Mode=::NewEnumerator2
    Sprite Size Mode=::NewEnumerator1
    Uniform Sprite Size Max=45
    Uniform Sprite Size Min=30
    UsePositionOffset=false
    Write Material Random=true
   )
   &PartitionParticles(Number Of Partitions=$Emitter.PortCount)
   &SkeletalMeshLocation(
    Bone / Socket Index=$Particles.PartitionParticles.Partition
    Bone / Socket Sampling Mode=::NewEnumerator5
    MaskSpawnGroup=false
    Mesh Sampling Type=::NewEnumerator6
    Skeletal Mesh=$System.SkeletalMesh
   )
   &ShapeLocation(
    MaskSpawnGroup=false
    Non Uniform Scale=<0.5 1 0.7>
    Offset=&AddVectorToPosition(Position=<0 0 0>, Vector=$User.DTCarOffset)
    Offset Coordinate Space=::Local
    Offset Mode=::NewEnumerator0
    Override Seed=false
    Sphere Radius=70
    UseOwnerScale=true
   )
   &AddVelocity(
    Velocity=&RandomRangeVector(Minimum=<-1100 0 0>, Maximum=<-1200 1 1>)
    Velocity Speed Scale=&SelectFloatFromArray(Direct Array Index=$User.GameModeID, Float Selection Array=DataInterface() {propertyValues {FloatData [1,1,0.7]}})
   )
   /SetVariables(Particles.VisibilityTag=$Emitter.SpawnGroup) {isSetParameters=true}
  ]
  particleUpdateScript [
   &ParticleState()
   &ScaleSpriteSize(Uniform Curve Sprite Scale=DataInterface() {propertyValues {Curve {keys [{tangentMode=RCTM_User, time=0, value=0.2}, {tangentMode=RCTM_User, time=0.1, value=1}]}}})
   &InheritSourceMovement(ApplyForce=false, ApplyPosition=true, ApplyPositionFalloff=false)
   &SolveForcesAndVelocity(Clamp Velocity=false, Limit Acceleration=false)
   &SubUVAnimation(
    Random Start Frame=true
    "Sprite Renderer "=DataInterface() {propertyValues {SpriteRenderer=&self:NE_VC_Boost_Sprite_Parent001.NiagaraSpriteRendererProperties_1}}
    UseEndFrame=false
    UseStartFrame=false
   )
   &DynamicMaterialParameters(
    Index 1 Param 1=&FloatFromCurve(FloatCurve=DataInterface() {propertyValues {Curve {keys [
         {interpMode=RCIM_Cubic, leaveTangentWeight=0.0667, tangentMode=RCTM_User, time=0, value=1}
         {arriveTangentWeight=0.1017, interpMode=RCIM_Cubic, leaveTangentWeight=0.1088, tangentMode=RCTM_User, time=0.2, value=0}
         {arriveTangentWeight=0.195, interpMode=RCIM_Cubic, leaveTangentWeight=0.2119, tangentMode=RCTM_User, time=0.5, value=0}
         {arriveTangent=1.045, arriveTangentWeight=0.1692, interpMode=RCIM_Cubic, leaveTangent=1.045, tangentMode=RCTM_User, time=0.9, value=1}
        ]}}})
    Module.Index1Float0Write=true
    Module.Index1Float1Write=true
    Module.Index1Float2Write=true
    Module.Index1Float3Write=true
    Write Parameter Index 0=::NewEnumerator0
    Write Parameter Index 1=::NewEnumerator1
   )
  ]
  rendererClasses [&NiagaraSpriteRendererProperties]
  renderers [
   {rendererClass=&NiagaraSpriteRendererProperties}
   {rendererClass=&NiagaraSpriteRendererProperties, rendererIndex=1}
  ]
 }
 NE_VC_Boost_Sprite_Parent002 {
  emitterUpdateScript [
   &EmitterState(Inactive Response=::NewEnumerator1, Life Cycle Mode=::NewEnumerator1, Loop Duration=1, UseLoopDelay=false)
   &VC_FilteredSocketCount()
   &SpawnRate(SpawnRate=16, Use Spawn Probability=false)
   &SpawnRate(SpawnGroup=1, SpawnRate=12, Use Spawn Probability=false) {module=SpawnRate001}
   &SpawnRate(Spawn Probability=0.5, SpawnGroup=2, SpawnRate=6, Use Spawn Probability=true) {module=SpawnRate002}
   &SpawnRate(Spawn Probability=0.6, SpawnGroup=3, SpawnRate=8, Use Spawn Probability=true) {module=SpawnRate003}
   &SpawnRate(SpawnGroup=4, SpawnRate=4, Use Spawn Probability=false) {module=SpawnRate004}
  ]
  particleSpawnScript [
   &VC_PaintedVariantsOverride() {enabled=false}
   &InitializeParticle(
    Color=&SelectLinearColorFromArray(
     Array Sampling Mode=::NewEnumerator1
     Direct Array Index=$User.GameModeID
     Color Selection Array=DataInterface() {propertyValues {ColorData [<rgba 0.7 0.7 0.7 1>, <rgba 0.9 0.9 0.9 1>, <rgba 0.9 0.9 0.9 1>]}}
     Random Seed=?
    )
    Lifetime=&MergeHelper_FloatSwitch(
     0=&RandomRangeFloat(Minimum=0.5, Maximum=0.7)
     1=&RandomRangeFloat(Minimum=0.4, Maximum=0.6)
     2=0.6
     3=0.4
     4=0.6
     Index=&SpawnGroup()
     Num=::NewEnumerator3
    )
    Mesh Scale Mode=::NewEnumerator1
    Mesh Uniform Scale Max=0.25
    Mesh Uniform Scale Min=0.14
    Sprite Rotation Mode=::NewEnumerator2
    Sprite Size Mode=::NewEnumerator3
    Uniform Sprite Size=&MergeHelper_FloatSwitch(
     0=0
     1=&RandomRangeFloat(Minimum=20, Maximum=35)
     2=&RandomRangeFloat(Minimum=70, Maximum=80)
     3=&RandomRangeFloat(Minimum=50, Maximum=65)
     4=&RandomRangeFloat(Minimum=155, Maximum=165)
     Index=&SpawnGroup()
     Num=::NewEnumerator3
    )
    UsePositionOffset=false
    Write Material Random=true
   )
   &PartitionParticles(Number Of Partitions=$Emitter.PortCount)
   &SkeletalMeshLocation(
    Bone / Socket Index=$Particles.PartitionParticles.Partition
    Bone / Socket Sampling Mode=::NewEnumerator5
    MaskSpawnGroup=false
    Mesh Sampling Type=::NewEnumerator6
    Skeletal Mesh=$System.SkeletalMesh
   )
   &ShapeLocation(
    MaskSpawnGroup=false
    Non Uniform Scale=<0.25 1 1>
    Offset=&AddVectorToPosition(
     Position=&MergeHelper_VectorSwitch(0=<-60 0 0>, 1=<-60 0 0>, 2=<-80 0 0>, 3=<-90 0 0>, 4=<-65 0 5>, Index=&SpawnGroup(), Num=::NewEnumerator3)
     Vector=$User.DTCarOffset
    )
    Offset Coordinate Space=::Local
    Offset Mode=::NewEnumerator0
    Override Seed=false
    Sphere Radius=&MergeHelper_FloatSwitch(0=40, 1=60, 2=10, 3=30, 4=45, Index=&SpawnGroup(), Num=::NewEnumerator3)
    Sphere Surface Distribution=0.2
    UseOwnerScale=true
   )
   &InitialMeshOrientation(
    Mesh Orientation Mode=::NewEnumerator7
    Rotation=&RandomRangeVector(Minimum=<0 0 0>, Maximum=<1 0 0>)
    Rotation Coordinate Space=::NewEnumerator0
    Use Rotation Vector=true
   )
   &AddVelocity(Velocity=&MergeHelper_VectorSwitch(
     0=<0 0 0>
     1=<0 0 0>
     2=&RandomRangeVector(Minimum=<-110 0 0>, Maximum=<-140 0 0>)
     3=&RandomRangeVector(Minimum=<-170 0 0>, Maximum=<-230 0 0>)
     4=<0 0 0>
     Index=&SpawnGroup()
     Num=::NewEnumerator3
    ))
   /SetVariables(Particles.VisibilityTag=$Emitter.SpawnGroup) {isSetParameters=true}
  ]
  particleUpdateScript [
   &ParticleState()
   &SolveForcesAndVelocity(Clamp Velocity=false, Limit Acceleration=false)
   &SubUVAnimation(
    End Frame Range Override=5
    "Sprite Renderer "=DataInterface() {propertyValues {SpriteRenderer=&self:NE_VC_Boost_Sprite_Parent002.NiagaraSpriteRendererProperties_0}}
    UseEndFrame=true
    UseStartFrame=true
   )
   /SetVariables(
    Particles.animation2:NiagaraFloat=&FloatFromCurve(FloatCurve=DataInterface() {propertyValues {Curve {keys [<at 0 6>,<at 1 9>]}}})
    Particles.animationmesh:Vector3f=&Multiply_VectorByFloat(
     Vector=$Particles.Scale
     Float=&FloatFromCurve(FloatCurve=DataInterface() {propertyValues {Curve {
         defaultValue=1
         keys [{arriveTangent=1.1695, interpMode=RCIM_Cubic, leaveTangent=1.1695, tangentMode=RCTM_User, time=0, value=0}, <at 0.5 1 RCIM_Cubic>]
        }}})
    )
    Particles.animationsize:Vector2f=&Multiply_Vector2D_ByFloat(
     Float=&FloatFromCurve(FloatCurve=DataInterface() {propertyValues {Curve {keys [
          {arriveTangent=2.6909, interpMode=RCIM_Cubic, leaveTangent=2.6909, leaveTangentWeight=0.1, tangentMode=RCTM_User, time=0, value=0.15}
          {
           arriveTangent=-1.7813
           arriveTangentWeight=0.1704
           interpMode=RCIM_Cubic
           leaveTangent=-1.7813
           leaveTangentWeight=0.1732
           tangentMode=RCTM_User
           time=0.6
           value=1.1
          }
          {arriveTangent=-0.4683, arriveTangentWeight=0.0438, interpMode=RCIM_Cubic, leaveTangent=-0.4683, tangentMode=RCTM_User, time=1, value=0.4}
         ]}}})
     Vector2D=$Particles.SpriteSize
    )
    Particles.animationsize1:Vector2f=&Multiply_Vector2D_ByFloat(
     Float=&FloatFromCurve(FloatCurve=DataInterface() {propertyValues {Curve {keys [
          {arriveTangent=0.1407, leaveTangentWeight=0.1667, time=0.6, value=0.9}
          {arriveTangent=0.6233, arriveTangentWeight=0.1667, leaveTangent=0.6233, time=1, value=1.1}
         ]}}})
     Vector2D=$Particles.SpriteSize
    )
    Particles.animationsize2:Vector2f=&Multiply_Vector2D_ByFloat(
     Float=&FloatFromCurve(FloatCurve=DataInterface() {propertyValues {Curve {keys [
          {arriveTangent=-0.0574, interpMode=RCIM_Cubic, leaveTangent=-0.0574, tangentMode=RCTM_User, time=0, value=1}
          {arriveTangent=-0.1621, interpMode=RCIM_Cubic, leaveTangent=-0.1621, tangentMode=RCTM_User, time=1, value=0.85}
         ]}}})
     Vector2D=$Particles.SpriteSize
    )
    Particles.animationsize3:Vector2f=&Multiply_Vector2D_ByFloat(
     Float=&FloatFromCurve(FloatCurve=DataInterface() {propertyValues {Curve {keys [
          {arriveTangent=0.1407, interpMode=RCIM_Cubic, leaveTangent=0.1407, leaveTangentWeight=0.1667, tangentMode=RCTM_User, time=0.6, value=1}
          {arriveTangent=0.6233, arriveTangentWeight=0.1667, interpMode=RCIM_Cubic, leaveTangent=0.6233, tangentMode=RCTM_User, time=1, value=1.3}
         ]}}})
     Vector2D=$Particles.SpriteSize
    )
   ) {isSetParameters=true}
   &DynamicMaterialParameters(
    Index 1 Param 1=&MergeHelper_FloatSwitch(
     0=&FloatFromCurve(FloatCurve=DataInterface() {propertyValues {Curve {
         defaultValue=1
         keys [
          {arriveTangent=-2.9248, interpMode=RCIM_Cubic, leaveTangent=-2.9248, leaveTangentWeight=0.2837, tangentMode=RCTM_User, time=0, value=1}
          {arriveTangentWeight=0.2981, interpMode=RCIM_Cubic, leaveTangentWeight=0.0745, tangentMode=RCTM_User, time=0.4, value=0}
          {arriveTangentWeight=0.5667, interpMode=RCIM_Cubic, leaveTangent=0.4918, leaveTangentWeight=0.7682, tangentMode=RCTM_Break, time=0.5, value=0}
          {arriveTangent=0.1091, arriveTangentWeight=0.1667, interpMode=RCIM_Cubic, tangentMode=RCTM_Break, time=1, value=1}
         ]
        }}})
     1=&FloatFromCurve(FloatCurve=DataInterface() {propertyValues {Curve {
         defaultValue=1
         keys [
          {arriveTangent=-7.1742, interpMode=RCIM_Cubic, leaveTangent=-7.1742, leaveTangentWeight=0.0667, tangentMode=RCTM_User, time=0, value=1}
          {arriveTangentWeight=0.0667, interpMode=RCIM_Cubic, leaveTangentWeight=0.4917, tangentMode=RCTM_Break, time=0.2, value=0}
          {arriveTangent=0.1905, interpMode=RCIM_Cubic, tangentMode=RCTM_Break, time=1, value=1}
         ]
        }}})
     2=0
     3=0
     4=0
     Index=&SpawnGroup()
     Num=::NewEnumerator3
    )
    Module.Index1Float0Write=true
    Module.Index1Float1Write=true
    Module.Index1Float2Write=false
    Module.Index1Float3Write=false
    Write Parameter Index 0=::NewEnumerator0
    Write Parameter Index 1=::NewEnumerator1
   )
  ]
  rendererClasses [&NiagaraMeshRendererProperties, &NiagaraSpriteRendererProperties]
  renderers [
   {rendererClass=&NiagaraMeshRendererProperties, rendererIndex=0}
   {rendererClass=&NiagaraSpriteRendererProperties, rendererIndex=1}
   {rendererClass=&NiagaraSpriteRendererProperties, rendererIndex=2}
   {rendererClass=&NiagaraSpriteRendererProperties, rendererIndex=3}
   {rendererClass=&NiagaraSpriteRendererProperties, rendererIndex=4}
  ]
 }
}
summary {
 systemName=NS_Courgette01
 emitters [
  {
   emitterName=NE_VC_Boost_Sprite_Parent
   rendererClasses [&NiagaraMeshRendererProperties, &NiagaraSpriteRendererProperties]
  }
  {emitterName=NE_VC_Boost_Sprite_Parent001, rendererClasses [&NiagaraSpriteRendererProperties]}
  {emitterName=NE_VC_Boost_Ribbon_Parent, rendererClasses [&NiagaraRibbonRendererProperties]}
  {
   emitterName=NE_VC_Boost_Sprite_Parent002
   rendererClasses [&NiagaraMeshRendererProperties, &NiagaraSpriteRendererProperties]
  }
 ]
 userVariables [
  User.bIsFrontendPreview {type=NiagaraBool}
  User.bIsPainted {type=NiagaraBool}
  User.ColorVariantID {type=NiagaraInt32}
  User.DTCarOffset {type=Vector3f, description="", defaultValue {struct=Vector3f, value=<0 0 0>}}
  User.GameModeID {type=NiagaraInt32}
  User.PaintedColor {type=LinearColor, description="", defaultValue {struct=LinearColor, value=<rgba 1 1 1 1>}}
  User.PaintedScalar {type=NiagaraFloat, defaultValue=1, description=""}
  User.PortCount {type=NiagaraInt32, defaultValue=2}
 ]
}
